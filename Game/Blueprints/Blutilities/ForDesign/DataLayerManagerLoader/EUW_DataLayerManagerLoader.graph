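# Editor utility widget, titled Data Layer Manager Loader, for converting the levels in a given directory to enable World Partition streaming support.
# The layout stacks a title, an instruction line, a directory text field and a Convert Levels button.
# The text field and button are inherited from the native FortDataLayerManagerLoader parent class.
# The event graph holds only unconnected default events, so the conversion logic is not in this Blueprint.
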EditorUtilityWidgetBlueprint /Game/Blueprints/Blutilities/ForDesign/DataLayerManagerLoader/EUW_DataLayerManagerLoader
graphs [EventGraph {nodes [
   Event PreConstruct()
   Event Construct()
   Event Tick()
  ]}]
variables []
widgets {
 info {
  inheritedWidgetCount=2
  parentClass=/Script/FortniteEditor.FortDataLayerManagerLoader
  rootWidgetClass=&HorizontalBox
  widgetCount=14
 }
 widgets [
  HorizontalBox_29 {widgetClassPath=&HorizontalBox, widget=HorizontalBox_29}
  VerticalBox_56 {
   widgetClassPath=&VerticalBox
   parent=HorizontalBox_29
   widget=VerticalBox_56
   slot {class=HorizontalBoxSlot, properties {size {sizeRule=Fill, value=1}}}
  }
  Border_70 {
   widgetClassPath=&Border
   parent=VerticalBox_56
   widget=Border_70
   properties {horizontalAlignment=HAlign_Center, brushColor=<rgba 0.0284 0.0284 0.0284 1>, padding=<ltrb 4 6 4 6>}
   slot {class=VerticalBoxSlot}
  }
  CommonTextBlock_332 {
   widgetClassPath=&CommonTextBlock
   parent=Border_70
   widget=CommonTextBlock_332
   properties {
    style=/FortUILibrary/Styles/Typography/Foundation/TextStyle-Base-XL-B.TextStyle-Base-XL-B_C
    text=Data Layer Manager Loader
    font {
     bForceMonospaced=false
     bMaterialIsStencil=false
     fontMaterial=null
     fontObject=&NotoSans
     letterSpacing=0
     monospacedWidth=1
     size=44
     skewAmount=0
     typefaceFontName=NotoSans-Bold
     outlineSettings {
      bApplyOutlineToDropShadows=false
      bMiteredCorners=false
      bSeparateFillAlpha=false
      outlineMaterial=null
      outlineSize=0
      outlineColor=<rgba 0 0 0 1>
     }
    }
    shadowOffset=<0 0>
   }
   slot {class=BorderSlot, properties {horizontalAlignment=HAlign_Center, padding=<ltrb 4 6 4 6>}}
  }
  Border_157 {
   widgetClassPath=&Border
   parent=VerticalBox_56
   widget=Border_157
   properties {horizontalAlignment=HAlign_Center, brushColor=<rgba 0.0284 0.0284 0.0284 1>, padding=<ltrb 4 6 4 6>}
   slot {class=VerticalBoxSlot}
  }
  CommonTextBlock_178 {
   widgetClassPath=&CommonTextBlock
   parent=Border_157
   widget=CommonTextBlock_178
   properties {
    style=/FortUILibrary/Styles/Typography/Foundation/TextStyle-Base-S.TextStyle-Base-S_C
    text="Enter a directory containing Levels to enable WP Streaming Support, then click the Convert button:"
    font {
     bForceMonospaced=false
     bMaterialIsStencil=false
     fontMaterial=null
     fontObject=&NotoSans
     letterSpacing=0
     monospacedWidth=1
     size=18
     skewAmount=0
     typefaceFontName=NotoSans-Regular
     outlineSettings {
      bApplyOutlineToDropShadows=false
      bMiteredCorners=false
      bSeparateFillAlpha=false
      outlineMaterial=null
      outlineSize=0
      outlineColor=<rgba 0 0 0 1>
     }
    }
    shadowOffset=<0 0>
   }
   slot {class=BorderSlot, properties {horizontalAlignment=HAlign_Center, padding=<ltrb 4 6 4 6>}}
  }
  Border_43 {
   widgetClassPath=&Border
   parent=VerticalBox_56
   widget=Border_43
   properties {brushColor=<rgba 0 0.162 0.7454 1>, padding=<ltrb 0 0 0 0>}
   slot {class=VerticalBoxSlot, properties {verticalAlignment=VAlign_Center, padding=<ltrb 0 2 0 2>}}
  }
  HorizontalBox_63 {
   widgetClassPath=&HorizontalBox
   parent=Border_43
   widget=HorizontalBox_63
   slot {class=BorderSlot, properties {padding=<ltrb 0 0 0 0>}}
  }
  SizeBox_1 {
   widgetClassPath=&SizeBox
   parent=HorizontalBox_63
   widget=SizeBox_1
   properties {bOverride_MinDesiredWidth=true, minDesiredWidth=120}
   slot {class=HorizontalBoxSlot, properties {padding=<ltrb 0 5 0 5>}}
  }
  CommonTextBlock_64 {
   widgetClassPath=&CommonTextBlock
   parent=SizeBox_1
   widget=CommonTextBlock_64
   properties {
    style=&TextStyle-Base-S_C
    text="Levels Directory: "
    toolTipText=Map that will be used by the automated nightly script to match the location gameplay tags set in the map marker component.
    visibility=Visible
    font {
     bForceMonospaced=false
     bMaterialIsStencil=false
     fontMaterial=null
     fontObject=&NotoSans
     letterSpacing=0
     monospacedWidth=1
     size=18
     skewAmount=0
     typefaceFontName=NotoSans-Regular
     outlineSettings {
      bApplyOutlineToDropShadows=false
      bMiteredCorners=false
      bSeparateFillAlpha=false
      outlineMaterial=null
      outlineSize=0
      outlineColor=<rgba 0 0 0 1>
     }
    }
    shadowOffset=<0 0>
   }
   slot {class=SizeBoxSlot, properties {horizontalAlignment=HAlign_Left, verticalAlignment=VAlign_Center, padding=<ltrb 10 0 10 0>}}
  }
  TextField_Directory {
   widgetClassPath=&EditorUtilityEditableTextBox
   bInherited=true
   bIsVariable=true
   parent=HorizontalBox_63
   widget=TextField_Directory
   slot {class=HorizontalBoxSlot, properties {padding=<ltrb 0 4 2 4>, size {sizeRule=Fill, value=1}}}
  }
  Border_290 {
   widgetClassPath=&Border
   parent=VerticalBox_56
   widget=Border_290
   properties {horizontalAlignment=HAlign_Center, brushColor=<rgba 0.0284 0.0284 0.0284 1>, padding=<ltrb 4 6 4 6>}
   slot {class=VerticalBoxSlot, properties {verticalAlignment=VAlign_Center}}
  }
  Button_Run {
   widgetClassPath=&EditorUtilityButton
   bInherited=true
   parent=Border_290
   widget=Button_Run
   slot {class=BorderSlot, properties {horizontalAlignment=HAlign_Center, padding=<ltrb 4 6 4 6>}}
  }
  CommonTextBlock_83 {
   widgetClassPath=&CommonTextBlock
   parent=Button_Run
   widget=CommonTextBlock_83
   properties {
    style=&TextStyle-Base_C
    text=Convert Levels
    font {
     bForceMonospaced=false
     bMaterialIsStencil=false
     fontMaterial=null
     fontObject=&NotoSans
     letterSpacing=0
     monospacedWidth=1
     size=24
     skewAmount=0
     typefaceFontName=NotoSans-Regular
     outlineSettings {
      bApplyOutlineToDropShadows=false
      bMiteredCorners=false
      bSeparateFillAlpha=false
      outlineMaterial=null
      outlineSize=0
      outlineColor=<rgba 0 0 0 1>
     }
    }
    shadowOffset=<0 0>
   }
   slot {class=ButtonSlot}
  }
 ]
}
